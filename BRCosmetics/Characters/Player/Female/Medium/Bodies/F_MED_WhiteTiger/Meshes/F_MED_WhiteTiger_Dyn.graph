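# Physics asset for the F_MED_WhiteTiger female player cosmetic body, defining secondary motion for its coat.
# Seven small capsule bodies on the back, left and right dyn_coat bones each constrain directly to spine_04 with limited swing and tight twist.
# Right-side dyn_coat_r_b has a narrower second swing limit than the other coat bones.
# Larger spine_02 to spine_04 capsules form the torso chain, joined by constraints with free swing and twist.
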PhysicsAsset /BRCosmetics/Characters/Player/Female/Medium/Bodies/F_MED_WhiteTiger/Meshes/F_MED_WhiteTiger_Dyn
bodies {
 dyn_coat_bk [{shapeType=Capsule, length=5, radius=2, shapeName=dyn_coat_bk_capsule, center=<5 0 0>, rotation=<rot 0 0 90>}]
 dyn_coat_l_a [{shapeType=Capsule, length=5, radius=2, shapeName=dyn_coat_l_a_capsule, center=<5 0 0>, rotation=<rot 0 0 90>}]
 dyn_coat_l_b [{shapeType=Capsule, length=5, radius=2, shapeName=dyn_coat_l_b_capsule, center=<5 0 0>, rotation=<rot 0 0 90>}]
 dyn_coat_l_c [{shapeType=Capsule, length=5, radius=2, shapeName=dyn_coat_l_c_capsule, center=<5 0 0>, rotation=<rot 0 0 90>}]
 dyn_coat_r_a [{shapeType=Capsule, length=5, radius=2, shapeName=dyn_coat_r_a_capsule, center=<5 0 0>, rotation=<rot 0 0 90>}]
 dyn_coat_r_b [{shapeType=Capsule, length=5, radius=2, shapeName=dyn_coat_r_b_capsule, center=<5 0 0>, rotation=<rot 0 0 90>}]
 dyn_coat_r_c [{shapeType=Capsule, length=5, radius=2, shapeName=dyn_coat_r_c_capsule, center=<5 0 0>, rotation=<rot 0 0 90>}]
 spine_02 [{shapeType=Capsule, length=5, radius=5}]
 spine_03 [{shapeType=Capsule, length=5, radius=5}]
 spine_04 [{shapeType=Capsule, length=5, radius=5}]
}
constraints [
 {bone1Name=dyn_coat_l_b, bone2Name=spine_04, swing1LimitDegrees=20, swing2LimitDegrees=60, twistLimitDegrees=5}
 {bone1Name=dyn_coat_l_a, bone2Name=spine_04, swing1LimitDegrees=20, swing2LimitDegrees=60, twistLimitDegrees=5}
 {bone1Name=dyn_coat_bk, bone2Name=spine_04, swing1LimitDegrees=20, swing2LimitDegrees=60, twistLimitDegrees=5}
 {bone1Name=dyn_coat_r_a, bone2Name=spine_04, swing1LimitDegrees=20, swing2LimitDegrees=60, twistLimitDegrees=5}
 {bone1Name=dyn_coat_r_b, bone2Name=spine_04, swing1LimitDegrees=20, swing2LimitDegrees=15, twistLimitDegrees=5}
 {
  bone1Name=spine_04
  bone2Name=spine_03
  swing1LimitDegrees=45
  swing1Motion=Free
  swing2LimitDegrees=45
  swing2Motion=Free
  twistLimitDegrees=45
  twistMotion=Free
 }
 {
  bone1Name=spine_03
  bone2Name=spine_02
  swing1LimitDegrees=45
  swing1Motion=Free
  swing2LimitDegrees=45
  swing2Motion=Free
  twistLimitDegrees=45
  twistMotion=Free
 }
 {bone1Name=dyn_coat_r_c, bone2Name=spine_04, swing1LimitDegrees=20, swing2LimitDegrees=60, twistLimitDegrees=5}
 {bone1Name=dyn_coat_l_c, bone2Name=spine_04, swing1LimitDegrees=20, swing2LimitDegrees=60, twistLimitDegrees=5}
]
