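# Physics asset for the F_MED_WhiteTiger female body, covering the dynamic right-side holster, holster strap and shoestring bones.
# Small capsule bodies form a chain from pelvis through right thigh to calf, where four shoestrings each attach separately.
# Holster and strap hang from MSTR anchor bodies constrained to thigh twist and pelvis respectively.
# The holster locks twist and one swing axis for hinge-like motion, while the strap rotates freely.
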PhysicsAsset /BRCosmetics/Characters/Player/Female/Medium/Bodies/F_MED_WhiteTiger/Meshes/F_MED_WhiteTiger_Dyn_Holster
bodies {
 calf_twist_02_r [{shapeType=Capsule, length=5, radius=5}]
 dyn_holster_r [{shapeType=Capsule, length=2, radius=1, center=<1.9851 0 0>, rotation=<rot 90 0 0>}]
 dyn_holster_r_MSTR [{shapeType=Capsule, length=0.505, radius=0.505, rotation=<rot 0 -90 0>}]
 dyn_holster_strap_r [{shapeType=Capsule, length=2, radius=1, center=<0.1 0 0>, rotation=<rot 90 0 0>}]
 dyn_holster_strap_r_MSTR [{shapeType=Capsule, length=0.505, radius=0.505, rotation=<rot 0 -90 0>}]
 dyn_shoestring_a [{shapeType=Capsule, length=2, radius=1, center=<0.3 0 0>, rotation=<rot 90 0 0>}]
 dyn_shoestring_b [{shapeType=Capsule, length=2, radius=1, center=<0.3 0 0>, rotation=<rot 90 0 0>}]
 dyn_shoestring_c [{shapeType=Capsule, length=2, radius=1, center=<1 0 0>, rotation=<rot 90 0 0>}]
 dyn_shoestring_d [{shapeType=Capsule, length=2, radius=1, center=<1 0 0>, rotation=<rot 90 0 0>}]
 pelvis [{shapeType=Capsule, length=1, radius=1}]
 thigh_r [{shapeType=Capsule, length=1, radius=1}]
 thigh_twist_01_r [{shapeType=Capsule, length=1, radius=1}]
}
constraints [
 {bone1Name=dyn_shoestring_b, bone2Name=calf_twist_02_r, swing1LimitDegrees=45, swing2LimitDegrees=45, twistLimitDegrees=45}
 {bone1Name=dyn_shoestring_d, bone2Name=calf_twist_02_r, swing1LimitDegrees=45, swing2LimitDegrees=45, twistLimitDegrees=45}
 {bone1Name=dyn_shoestring_c, bone2Name=calf_twist_02_r, swing1LimitDegrees=45, swing2LimitDegrees=45, twistLimitDegrees=45}
 {bone1Name=dyn_shoestring_a, bone2Name=calf_twist_02_r, swing1LimitDegrees=45, swing2LimitDegrees=45, twistLimitDegrees=45}
 {
  bone1Name=dyn_holster_r
  bone2Name=dyn_holster_r_MSTR
  swing1LimitDegrees=45
  swing1Motion=Locked
  swing2LimitDegrees=20
  twistLimitDegrees=45
  twistMotion=Locked
 }
 {
  bone1Name=dyn_holster_strap_r
  bone2Name=dyn_holster_strap_r_MSTR
  swing1LimitDegrees=45
  swing1Motion=Free
  swing2LimitDegrees=45
  swing2Motion=Free
  twistLimitDegrees=45
  twistMotion=Free
 }
 {
  bone1Name=calf_twist_02_r
  bone2Name=thigh_r
  swing1LimitDegrees=45
  swing1Motion=Free
  swing2LimitDegrees=45
  swing2Motion=Free
  twistLimitDegrees=45
  twistMotion=Free
 }
 {
  bone1Name=dyn_holster_r_MSTR
  bone2Name=thigh_twist_01_r
  swing1LimitDegrees=45
  swing1Motion=Free
  swing2LimitDegrees=45
  swing2Motion=Free
  twistLimitDegrees=45
  twistMotion=Free
 }
 {
  bone1Name=dyn_holster_strap_r_MSTR
  bone2Name=pelvis
  swing1LimitDegrees=45
  swing1Motion=Free
  swing2LimitDegrees=45
  swing2Motion=Free
  twistLimitDegrees=45
  twistMotion=Free
 }
 {
  bone1Name=thigh_twist_01_r
  bone2Name=thigh_r
  swing1LimitDegrees=45
  swing1Motion=Free
  swing2LimitDegrees=45
  swing2Motion=Free
  twistLimitDegrees=45
  twistMotion=Free
 }
 {
  bone1Name=thigh_r
  bone2Name=pelvis
  swing1LimitDegrees=45
  swing1Motion=Free
  swing2LimitDegrees=45
  swing2Motion=Free
  twistLimitDegrees=45
  twistMotion=Free
 }
]
