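# Post-style AnimBP for the F_MED_WhiteTiger female body, adding corrective and secondary motion to the pose copied from the attached parent mesh.
# Thigh volume bones half-follow thigh rotation, then RigidBody nodes simulate dynamic bones relative to spine_01 and the right thigh.
# Crouch events toggle an alpha for a left thigh-volume offset and a right holster-strap constraint.
# The holster RigidBody and master-bone constraints are not wired to the output.
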AnimBlueprint /BRCosmetics/Characters/Player/Female/Medium/Bodies/F_MED_WhiteTiger/Meshes/F_MED_WhiteTiger_AnimBP
target_skeleton=&/F_MED_WhiteTiger_Skeleton
graphs [
 AnimGraph {nodes [
   n1 = "Constraint - dyn_holster_r_MSTR (PPP)"!(ComponentPose, Alpha=0, ConstraintWeights_0=0.0217, ConstraintWeights_1:float=0.0761, ConstraintWeights_2:float=0.9022) -> (Pose) {
    alphaBoolBlend {blendOption=Linear}
    boneToModify {boneName=dyn_holster_r_MSTR}
    constraintSetup [{targetBone {boneName=thigh_r}}, {targetBone {boneName=pelvis}}, {targetBone {boneName=thigh_twist_01_r}}]
    constraintWeights [1,1,1]
   }
   "Constraint - dyn_holster_r_MSTR (PP)"!(ComponentPose=@n1, Alpha=0, ConstraintWeights_0=0.3262, ConstraintWeights_1:float=0.6738) -> (Pose) {
    alphaBoolBlend {blendOption=Linear}
    boneToModify {boneName=dyn_holster_r_MSTR}
    constraintSetup [{targetBone {boneName=pelvis}}, {targetBone {boneName=thigh_twist_01_r}}]
    constraintWeights [1,1]
   }
   RigidBody() {
    bClampLinearTranslationLimitToRefPose=true
    bOverrideWorldGravity=true
    overridePhysicsAsset=&/F_MED_WhiteTiger_Dyn_Holster
    simulationSpace=BaseBoneSpace
    alphaBoolBlend {blendOption=Linear}
    baseBoneRef {boneName=pelvis}
    componentAppliedLinearAccClamp=<2000 2000 2000>
    componentLinearAccScale=<0.4 0.4 0.4>
    componentLinearVelScale=<0.4 0.4 0.4>
    overrideWorldGravity {z=-980}
   }
   n2 = "Constraint - thigh_volume_l (R)"(ComponentPose=Local To Component(LocalPose=Copy Pose From Mesh() {bUseAttachedParent=true}), Alpha=0.5, ConstraintWeights_0=1) {
    alphaBoolBlend {blendOption=Linear}
    boneToModify {boneName=thigh_volume_l}
    constraintSetup [{transformType=Rotation, targetBone {boneName=thigh_l}}]
    constraintWeights [1]
   }
   n3 = "Constraint - thigh_volume_r (R)"(ComponentPose=@n2, Alpha=0.5, ConstraintWeights_0=1) {
    alphaBoolBlend {blendOption=Linear}
    boneToModify {boneName=thigh_volume_r}
    constraintSetup [{transformType=Rotation, targetBone {boneName=thigh_r}}]
    constraintWeights [1]
   }
   n4 = Get crouch alpha() -> (crouch alpha:double)
   n5 = "Transform (Modify) Bone - Bone: thigh_volume_l"(ComponentPose=@n3, Alpha=@n4, Translation="-0.000,-1.379,1.177", Rotation="0,0,0", Scale="1,1,1") {
    scaleMode=BMM_Additive
    translationMode=BMM_Additive
    translationSpace=BCS_BoneSpace
    alphaBoolBlend {blendOption=Linear}
    boneToModify {boneName=thigh_volume_l}
    translation {y=-1.379, z=1.177}
   }
   n6 = RigidBody(ComponentPose=@n5) {
    bClampLinearTranslationLimitToRefPose=true
    bOverrideWorldGravity=true
    overridePhysicsAsset=&/F_MED_WhiteTiger_Dyn
    simulationSpace=BaseBoneSpace
    alphaBoolBlend {blendOption=Linear}
    baseBoneRef {boneName=spine_01}
    componentAppliedLinearAccClamp=<2000 2000 2000>
    componentLinearAccScale=<0.4 0.4 0.4>
    componentLinearVelScale=<0.4 0.4 0.4>
    overrideWorldGravity {z=-980}
   }
   n7 = RigidBody(ComponentPose=@n6) {
    bClampLinearTranslationLimitToRefPose=true
    bOverrideWorldGravity=true
    overridePhysicsAsset=&/F_MED_WhiteTiger_Dyn_Legs
    simulationSpace=BaseBoneSpace
    alphaBoolBlend {blendOption=Linear}
    baseBoneRef {boneName=thigh_r}
    componentAppliedLinearAccClamp=<2000 2000 2000>
    componentLinearAccScale=<0.4 0.4 0.4>
    componentLinearVelScale=<0.4 0.4 0.4>
    overrideWorldGravity {z=-980}
   }
   n8 = "Constraint - dyn_holster_strap_r (PP)"!(ComponentPose=@n7, Alpha=@n4, ConstraintWeights_0=0.4, ConstraintWeights_1:float=0.6) -> (Pose) {
    alphaBoolBlend {blendOption=Linear}
    boneToModify {boneName=dyn_holster_strap_r}
    constraintSetup [{targetBone {boneName=thigh_r}}, {targetBone {boneName=dyn_holster_strap_r_MSTR}}]
    constraintWeights [1,1]
   }
   Output Pose(Result=Component To Local(ComponentPose=@n8)) {layerGroup=None}
   Break Fort Anim Input Skydiving External Force(FortAnimInput_SkydivingExternalForce=Get SkydivingExternalForceInput()) -> (
    bUseSkydivingVectorForce=~
    SkydivingVectorForce=~
    HeadToPelvisDirection=~
    FloatingMultiplier=~
    FloatingAdditive=~
    DivingMultiplier=~
    DivingAdditive=~
    ParachutingMultiplier=~
    ParachutingAdditive=~
   )
  ]}
 EventGraph {nodes [
   Event OnBeginCrouching()
   > Set crouch alpha(crouch alpha:double=1) -> (Output_Get:double)
   Event OnEndCrouching()
   > Set crouch alpha(crouch alpha=0)
   TryGetPawnOwner()
   Event BlueprintUpdateAnimation()
  ]}
]
variables [crouch alpha {type=number}]
