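# Physics asset for the FaceAcc mesh of the White Tiger medium female player body, defining physics bodies for its dynamic hair.
# Seven three-bone hair chains around the front, sides and back are constrained in series from the head, with limited swing and very tight twist, slightly looser on the right-back chain.
# Center and right back chains use boxes, the others capsules.
# Head, neck and clavicle capsules supply body volumes, with head-to-neck angular motion left free.
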PhysicsAsset /BRCosmetics/Characters/Player/Female/Medium/Bodies/F_MED_WhiteTiger/Meshes/Parts/F_MED_White_Tiger_FaceAcc_Dyn
bodies {
 clavicle_l [{shapeType=Capsule, length=6, radius=9.25, center=<3.5169 -1.0045 -0.9029>, rotation=<rot 88.3167 -14.9967 -19.9711>}]
 clavicle_r [{shapeType=Capsule, length=6, radius=9.25, center=<-4.3337 0.9332 1.4407>, rotation=<rot 85.2589 -168.0866 -173.0303>}]
 dyn_hair_bk_a [{shapeType=Box, extentX=3.5, extentY=4, extentZ=2, shapeName=dyn_hair_bk_a_box, center=<1 0 0.574>}]
 dyn_hair_bk_b [{shapeType=Box, extentX=5, extentY=5, extentZ=2, shapeName=dyn_hair_bk_b_box, center=<1 0 -1>}]
 dyn_hair_bk_c [{shapeType=Box, extentX=5, extentY=5, extentZ=2, shapeName=dyn_hair_bk_c_box, center=<1 0 -1>}]
 dyn_hair_bk_l_a [{shapeType=Capsule, length=3, radius=1.5, shapeName=dyn_hair_bk_l_a_capsule, center=<0.5 0 0.25>, rotation=<rot 0 0 90>}]
 dyn_hair_bk_l_b [{shapeType=Capsule, length=3, radius=1.5, shapeName=dyn_hair_bk_l_b_capsule, center=<3 0 0.25>, rotation=<rot 0 0 90>}]
 dyn_hair_bk_l_c [{shapeType=Capsule, length=3, radius=1.5, shapeName=dyn_hair_bk_l_c_capsule, center=<3 0 0.25>, rotation=<rot 0 0 90>}]
 dyn_hair_bk_r_a [{shapeType=Box, extentX=3.5, extentY=6, extentZ=2, shapeName=dyn_hair_bk_r_a_box, center=<0.8167 0 1.2431>, rotation=<rot 19.9998 -0.0001 0>}]
 dyn_hair_bk_r_b [{
   shapeType=Box
   extentX=5
   extentY=6
   extentZ=2
   shapeName=dyn_hair_bk_r_b_box
   center=<0.9417 0.0574 0.5289>
   rotation=<rot -0.0003 -0.0002 10.0001>
  }]
 dyn_hair_bk_r_c [{
   shapeType=Box
   extentX=5
   extentY=6
   extentZ=2
   shapeName=dyn_hair_bk_r_c_box
   center=<0.9417 0.0574 0.5289>
   rotation=<rot -0.0003 -0.0002 10.0001>
  }]
 dyn_hair_l_bk_a [{shapeType=Capsule, length=3, radius=1.5, shapeName=dyn_hair_l_bk_a_capsule, center=<0.5 0 0.25>, rotation=<rot 0 0 90>}]
 dyn_hair_l_bk_b [{shapeType=Capsule, length=3, radius=1.5, shapeName=dyn_hair_l_bk_b_capsule, center=<3 0 0.25>, rotation=<rot 0 0 90>}]
 dyn_hair_l_bk_c [{shapeType=Capsule, length=3, radius=1.5, shapeName=dyn_hair_l_bk_c_capsule, center=<3 0 0.25>, rotation=<rot 0 0 90>}]
 dyn_hair_l_fr_a [{shapeType=Capsule, length=3, radius=1.5, shapeName=dyn_hair_l_fr_a_capsule, center=<0.5 0 0>, rotation=<rot 0 0 90>}]
 dyn_hair_l_fr_b [{shapeType=Capsule, length=3, radius=1.5, shapeName=dyn_hair_l_fr_b_capsule, center=<0.5 0 0>, rotation=<rot 0 0 90>}]
 dyn_hair_l_fr_c [{shapeType=Capsule, length=3, radius=1.5, shapeName=dyn_hair_l_fr_c_capsule, center=<0.5 0 0>, rotation=<rot 0 0 90>}]
 dyn_hair_r_bk_a [{
   shapeType=Box
   extentX=3.5
   extentY=6
   extentZ=2
   shapeName=dyn_hair_r_bk_a_box
   center=<0.9998 2.0874 0.8733>
   rotation=<rot 0.0023 0.0515 29.9981>
  }]
 dyn_hair_r_bk_b [{
   shapeType=Box
   extentX=5
   extentY=4.5
   extentZ=2
   shapeName=dyn_hair_r_bk_b_box
   center=<0.9999 1.706 -0.1321>
   rotation=<rot 0 -0.0001 24.9993>
  }]
 dyn_hair_r_bk_c [{
   shapeType=Box
   extentX=5
   extentY=4.5
   extentZ=2
   shapeName=dyn_hair_r_bk_c_box
   center=<0.9999 1.0725 -1.2295>
   rotation=<rot 0 -0.0001 24.9993>
  }]
 dyn_hair_r_fr_a [{shapeType=Capsule, length=3, radius=1.5, shapeName=dyn_hair_r_fr_a_capsule, center=<0.5 0 0>, rotation=<rot 90 0 0>}]
 dyn_hair_r_fr_b [{shapeType=Capsule, length=3, radius=1.5, shapeName=dyn_hair_r_fr_b_capsule, center=<0.5 0 0>, rotation=<rot 90 0 0>}]
 dyn_hair_r_fr_c [{shapeType=Capsule, length=3, radius=1.5, shapeName=dyn_hair_r_fr_c_capsule, center=<0.5 0.0001 0>, rotation=<rot 90 0 0>}]
 head [{shapeType=Capsule, length=11.7474, radius=5.7403, center=<-2.2245 0.2139 -0.3773>, rotation=<rot 90 90 79.9999>}]
 neck_01 [
  {shapeType=Capsule, length=6.75, radius=5.25, shapeName=neck_01_capsule, center=<0.9425 -1.974 0.0001>, rotation=<rot -0.0021 -10.0027 0>}
  {shapeType=Capsule, length=9.5, radius=4.5, shapeName=neck_01_capsule, center=<1.3499 2.0243 0>}
 ]
 neck_02 [{shapeType=Capsule, length=4.25, radius=5.25, shapeName=neck_01_capsule, center=<0.9425 -1.974 0.0001>, rotation=<rot -0.0021 -10.0027 0>}]
}
constraints [
 {bone1Name=dyn_hair_l_fr_a, bone2Name=head, swing1LimitDegrees=20, swing2LimitDegrees=10, twistLimitDegrees=1}
 {bone1Name=dyn_hair_l_fr_b, bone2Name=dyn_hair_l_fr_a, swing1LimitDegrees=20, swing2LimitDegrees=10, twistLimitDegrees=1}
 {bone1Name=dyn_hair_l_bk_a, bone2Name=head, swing1LimitDegrees=20, swing2LimitDegrees=45, twistLimitDegrees=1}
 {bone1Name=dyn_hair_l_bk_b, bone2Name=dyn_hair_l_bk_a, swing1LimitDegrees=20, swing2LimitDegrees=45, twistLimitDegrees=1}
 {bone1Name=dyn_hair_r_fr_a, bone2Name=head, swing1LimitDegrees=45, swing2LimitDegrees=12, twistLimitDegrees=1}
 {bone1Name=dyn_hair_r_fr_b, bone2Name=dyn_hair_r_fr_a, swing1LimitDegrees=45, swing2LimitDegrees=12, twistLimitDegrees=1}
 {bone1Name=dyn_hair_r_fr_c, bone2Name=dyn_hair_r_fr_b, swing1LimitDegrees=45, swing2LimitDegrees=12, twistLimitDegrees=1}
 {bone1Name=dyn_hair_l_fr_c, bone2Name=dyn_hair_l_fr_b, swing1LimitDegrees=20, swing2LimitDegrees=10, twistLimitDegrees=1}
 {bone1Name=dyn_hair_l_bk_c, bone2Name=dyn_hair_l_bk_b, swing1LimitDegrees=20, swing2LimitDegrees=45, twistLimitDegrees=1}
 {bone1Name=dyn_hair_bk_l_a, bone2Name=head, swing1LimitDegrees=20, swing2LimitDegrees=45, twistLimitDegrees=1}
 {bone1Name=dyn_hair_bk_l_b, bone2Name=dyn_hair_bk_l_a, swing1LimitDegrees=20, swing2LimitDegrees=45, twistLimitDegrees=1}
 {bone1Name=dyn_hair_bk_l_c, bone2Name=dyn_hair_bk_l_b, swing1LimitDegrees=20, swing2LimitDegrees=45, twistLimitDegrees=1}
 {bone1Name=dyn_hair_bk_a, bone2Name=head, swing1LimitDegrees=20, swing2LimitDegrees=45, twistLimitDegrees=1}
 {bone1Name=dyn_hair_bk_b, bone2Name=dyn_hair_bk_a, swing1LimitDegrees=20, swing2LimitDegrees=45, twistLimitDegrees=1}
 {bone1Name=dyn_hair_bk_c, bone2Name=dyn_hair_bk_b, swing1LimitDegrees=20, swing2LimitDegrees=45, twistLimitDegrees=1}
 {bone1Name=dyn_hair_bk_r_a, bone2Name=head, swing1LimitDegrees=15, swing2LimitDegrees=45, twistLimitDegrees=1}
 {bone1Name=dyn_hair_bk_r_b, bone2Name=dyn_hair_bk_r_a, swing1LimitDegrees=15, swing2LimitDegrees=45, twistLimitDegrees=1}
 {bone1Name=dyn_hair_bk_r_c, bone2Name=dyn_hair_bk_r_b, swing1LimitDegrees=15, swing2LimitDegrees=45, twistLimitDegrees=1}
 {bone1Name=dyn_hair_r_bk_a, bone2Name=head, swing1LimitDegrees=25, swing2LimitDegrees=45, twistLimitDegrees=10}
 {bone1Name=dyn_hair_r_bk_b, bone2Name=dyn_hair_r_bk_a, swing1LimitDegrees=25, swing2LimitDegrees=45, twistLimitDegrees=10}
 {bone1Name=dyn_hair_r_bk_c, bone2Name=dyn_hair_r_bk_b, swing1LimitDegrees=25, swing2LimitDegrees=45, twistLimitDegrees=10}
 {
  bone1Name=head
  bone2Name=neck_02
  swing1LimitDegrees=45
  swing1Motion=Free
  swing2LimitDegrees=45
  swing2Motion=Free
  twistLimitDegrees=45
  twistMotion=Free
 }
 {
  bone1Name=neck_02
  bone2Name=neck_01
  swing1LimitDegrees=45
  swing1Motion=Free
  swing2LimitDegrees=45
  swing2Motion=Free
  twistLimitDegrees=45
  twistMotion=Free
 }
]
